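# Shared Fortnite player Control Rig that procedurally builds controls for extra skeleton joints beyond the reference body and face sets.
# Construction imports the skeleton, skips reference names (minus include_joints) and dfrm_ joints, then spawns hierarchy-matched controls with _Direct children, linked through metadata.
# Forward solve drives joints from Direct controls and applies global_ctrl's display scale and visibility channels.
# Inverse solve copies joint transforms to main controls, resetting display scales.
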ControlRigBlueprint /BRCosmetics/Characters/Player/Common/SharedRigs/Fortnite_ExtraJoint_CtrlRig
graphs [
 ContainedGraph {nodes [
   n1 = RigVMFunctionEntryNode()
   > SequenceAggregate = RigVMAggregateNode()
   n2 = RigVMDispatchNode(Array=(), ExecuteContext=@SequenceAggregate.A) {template="DISPATCH_RigVMDispatch_ArrayReset(io Array)"}
   > n3 = RigVMDispatchNode(Array=@n1.Set A) {template="DISPATCH_RigVMDispatch_ArrayIterator(in Array,out Element,out Index,out Count,out Ratio)"}
   Branch_1_RigVMUnitNode = &RigVMFunction_ControlFlowBranch(
    Condition=RigVMDispatchNode(Element=@n3.Element, Array=@n1.Set B) {template="DISPATCH_RigVMDispatch_ArrayFind(in Array,in Element,out Index,out Success)"}.Success
    ExecuteContext=@n3.ExecuteContext
   )
   n4 = RigVMDispatchNode(Element=@n3.Element, Array=@n2.Array, ExecuteContext=@Branch_1_RigVMUnitNode.False) {template="DISPATCH_RigVMDispatch_ArrayAdd(io Array,in Element,out Index)"}
   RigVMFunctionReturnNode(Difference=@n4.Array, ExecuteContext=@SequenceAggregate.C)
  ]}
 ContainedGraph {nodes [
   n1 = RigVMFunctionEntryNode()
   SequenceAggregate = &RigVMFunction_Sequence(ExecuteContext=@n1.ExecuteContext)
   SequenceAggregate_1 = &RigVMFunction_Sequence(ExecuteContext=@SequenceAggregate.B)
   RigVMFunctionReturnNode(A=@SequenceAggregate.A, B=@SequenceAggregate_1.A, C=@SequenceAggregate_1.B)
  ]}
 RigVMFunctionLibrary {nodes [Set Difference = RigVMCollapseNode(ExecuteContext=(), Set A=(), Set B=())]}
 RigVMModel {nodes [
   n1 = /Script/ControlRig.RigUnit_InverseExecution() {is=struct RigVMUnitNode}
   n2 = RigVMDispatchNode(ExecuteContext=@n1.ExecutePin, Array=RigVMVariableNode() {variable=RiggedJoints}.Value) {template="DISPATCH_RigVMDispatch_ArrayIterator(in Array,out Element,out Index,out Count,out Ratio)"}
   n3 = RigVMDispatchNode(Default=(Type=None,Name="None"), Name=Control, NameSpace=Self, Item=@n2.Element) {template="DISPATCH_RigDispatch_GetMetadata(in Item,in Name,in NameSpace,in Default,out Value,out Found)"}
   Branch_4_RigVMUnitNode = &RigVMFunction_ControlFlowBranch(ExecuteContext=@n2.ExecuteContext, Condition=@n3.Found)
   n4 = /Script/ControlRig.RigUnit_HierarchyGetParent(bDefaultParent=True, Child=@n3.Value) {is=struct RigVMUnitNode}
   &RigUnit_SetTransform(
    Weight=1
    bInitial=false
    bPropagateToChildren=true
    ExecutePin=@Branch_4_RigVMUnitNode.True
    Value=&RigUnit_GetTransform(bInitial=false, Item=@n2.Element).Transform
    Item=@n4.Parent
   )
   > SetAnimationChannel = /Script/ControlRig.RigUnit_SetFloatAnimationChannel(Channel=Control Display Scale, Value=1, bInitial=False, Control=@n3.Value.Name) {is=struct RigVMUnitNode, template="SetAnimationChannel::Execute(in Value,in Control,in Channel,in bInitial)"}
   > &RigUnit_SetFloatAnimationChannel(Channel=Control Display Scale, Value=1, bInitial=False, Control=@n4.Parent.Name) {template="SetAnimationChannel::Execute(in Value,in Control,in Channel,in bInitial)"}
   > &RigUnit_SetFloatAnimationChannel(Channel=Control Display Scale, Value=1, bInitial=False, Control=RigVMVariableNode() {variable=GlobalSettingsCtrl}.Value.Name) {template="SetAnimationChannel::Execute(in Value,in Control,in Channel,in bInitial)"}
   n5 = &RigUnit_PrepareForExecution()
   n6 = &RigUnit_HierarchyImportFromSkeleton(NameSpace=None, bIncludeMeshSockets=False, bIncludeVirtualBones=True, ExecutePin=@n5.ExecutePin)
   > RigVMDispatchNode(Array=RigVMVariableNode() {variable=RiggedJoints}.Value) {template="DISPATCH_RigVMDispatch_ArrayReset(io Array)"}
   > SequenceAggregate = RigVMAggregateNode()
   n7 = &RigVMFunction_ControlFlowBranch(Condition=true, ExecuteContext=@SequenceAggregate.A)
   RigVMDispatchNode(
    ExecuteContext=@n7.True
    Array=RigVMVariableNode() {variable=exclude_joints}.Value
    Other=RigVMVariableNode() {variable=REF_Body_Skeleton_Names}.Value
   ) {template="DISPATCH_RigVMDispatch_ArrayUnion(io Array,in Other)"}
   RigVMFunction_ControlFlowBranch = &RigVMFunction_ControlFlowBranch(Condition=true, ExecuteContext=@SequenceAggregate.B)
   DISPATCH_RigVMDispatch_ArrayUnion = RigVMDispatchNode(
    ExecuteContext=@RigVMFunction_ControlFlowBranch.True
    Other=RigVMVariableNode() {variable=REF_Face_Skeleton_Names}.Value
    Array=RigVMVariableNode() {variable=exclude_joints}.Value
   ) {template="DISPATCH_RigVMDispatch_ArrayUnion(io Array,in Other)"}
   Set Difference_1 = RigVMFunctionReferenceNode(
    Set A=RigVMVariableNode() {variable=exclude_joints}.Value
    Set B=RigVMVariableNode() {variable=include_joints}.Value
    ExecuteContext=@SequenceAggregate.C
   )
   > RigVMVariableNode(Value=@"Set Difference_1".Difference) {variable=exclude_joints}
   > SpawnControl = /Script/ControlRig.RigUnit_HierarchyAddControlTransform(
    InitialValue=(Rotation=(X=0,Y=0,Z=0,W=1),Translation=(X=0,Y=0,Z=0),Scale3D=(X=1,Y=1,Z=1))
    Name=global_ctrl
    OffsetSpace=LocalSpace
    OffsetTransform=(Rotation=(X=0,Y=0,Z=0,W=1),Translation=(X=0,Y=0,Z=0),Scale3D=(X=1,Y=1,Z=1))
    Settings="(Shape=(bVisible=false,Name=\"Default\",Color=(R=1,G=0,B=0,A=1),Transform=(Rotation=(X=0,Y=0,Z=0,W=1),Translation=(X=0,Y=0,Z=0),Scale3D=(X=1,Y=1,Z=1))),Proxy=(bIsProxy=False,DrivenControls=((Type=None,Name=\"None\")),ShapeVisibility=BasedOnSelection),DisplayNa ...(truncated)"
    Parent=RigVMDispatchNode(Index=0, Array=@n6.Items) {template="DISPATCH_RigVMDispatch_ArrayGetAtIndex(in Array,in Index,out Element)"}.Element
    Settings.Shape=RigVMVariableNode() {variable=Global Control Settings}.Value
   ) {
    is=struct RigVMUnitNode
    template="SpawnControl::Execute(in InitialValue,in Settings,in OffsetTransform,in OffsetSpace,in Parent,in Name,out Item)"
   }
   > /Script/ControlRig.RigUnit_HierarchyAddAnimationChannelScaleFloat(
    InitialValue=1
    LimitsEnabled=(Enabled=(bMinimum=True,bMaximum=True))
    MaximumValue=20
    MinimumValue=0
    Name=Control Display Scale
    Parent=@SpawnControl.Item
   ) {
    is=struct RigVMUnitNode
    template="SpawnScaleAnimationChannel::Execute(in InitialValue,in MinimumValue,in MaximumValue,in LimitsEnabled,in Parent,in Name,out Item)"
   }
   > /Script/ControlRig.RigUnit_HierarchyAddAnimationChannelBool(InitialValue=true, LimitsEnabled=(), MaximumValue=True, MinimumValue=False, Name=Control Visibility, Parent=@SpawnControl.Item) {
    is=struct RigVMUnitNode
    template="SpawnAnimationChannel::Execute(in InitialValue,in MinimumValue,in MaximumValue,in LimitsEnabled,in Parent,in Name,out Item)"
   }
   > &RigUnit_HierarchyAddAnimationChannelBool(InitialValue=true, LimitsEnabled=(), MaximumValue=True, MinimumValue=False, Name=Direct Control Visibility, Parent=@SpawnControl.Item) {template="SpawnAnimationChannel::Execute(in InitialValue,in MinimumValue,in MaximumValue,in LimitsEnabled,in Parent,in Name,out Item)"}
   > RigVMVariableNode(Value=@SpawnControl.Item) {variable=GlobalSettingsCtrl}
   > n8 = RigVMDispatchNode(Array=@n6.Items) {template="DISPATCH_RigVMDispatch_ArrayIterator(in Array,out Element,out Index,out Count,out Ratio)"}
   DISPATCH_RigDispatch_ToString = RigVMDispatchNode(Value=@n8.Element.Name) {template="DISPATCH_RigDispatch_ToString(in Value,out Result)"}
   n9 = /Script/RigVM.RigVMFunction_StringStartsWith(Start=dfrm_, Name=@DISPATCH_RigDispatch_ToString.Result) {is=struct RigVMUnitNode, template="StartsWith::Execute(in Name,in Start,out Result)"}
   n10 = RigVMDispatchNode(Element=@DISPATCH_RigDispatch_ToString.Result, Array=RigVMVariableNode() {variable=exclude_joints}.Value) {template="DISPATCH_RigVMDispatch_ArrayFind(in Array,in Element,out Index,out Success)"}
   Branch_RigVMUnitNode = &RigVMFunction_ControlFlowBranch(
    ExecuteContext=@n8.ExecuteContext
    Condition=&RigVMFunction_MathBoolAnd(B=&RigVMFunction_MathBoolNot(Value=@n9.Result).Result, A=&RigVMFunction_MathBoolNot(Value=@n10.Success).Result).Result
   )
   RigVMDispatchNode(Element=@n8.Element, ExecuteContext=@Branch_RigVMUnitNode.True, Array=RigVMVariableNode() {variable=RiggedJoints}.Value) {template="DISPATCH_RigVMDispatch_ArrayAdd(io Array,in Element,out Index)"}
   n11 = &RigUnit_HierarchyGetParent(bDefaultParent=True, Child=@n8.Element)
   n12 = RigVMDispatchNode(Default=(Type=Bone,Name="None"), Name=Parent, NameSpace=Self, Item=@n11.Parent) {template="DISPATCH_RigDispatch_GetMetadata(in Item,in Name,in NameSpace,in Default,out Value,out Found)"}
   > n13 = &RigUnit_HierarchyAddControlTransform(
    InitialValue=(Rotation=(X=0,Y=0,Z=0,W=1),Translation=(X=0,Y=0,Z=0),Scale3D=(X=1,Y=1,Z=1))
    OffsetSpace=LocalSpace
    Name=@n8.Element.Name
    OffsetTransform=&RigUnit_GetTransform(Space=LocalSpace, bInitial=true, Item=@n8.Element).Transform
    Parent=RigVMDispatchNode(Condition=@n12.Found, True=@n12.Value, False=@n11.Parent) {template="DISPATCH_RigVMDispatch_If(in Condition,in True,in False,out Result)"}.Result
    Settings=RigVMVariableNode() {variable=Control Settings}.Value
   ) {template="SpawnControl::Execute(in InitialValue,in Settings,in OffsetTransform,in OffsetSpace,in Parent,in Name,out Item)"}
   > RigVMDispatchNode(Name=Parent, NameSpace=Self, Item=@n8.Element, Value=@n13.Item) {template="DISPATCH_RigDispatch_SetMetadata(in Item,in Name,in NameSpace,in Value,out Success)"}
   > RigVMDispatchNode(
    Name=Control Display Scale Init
    NameSpace=Self
    Value=/Script/ControlRig.RigUnit_GetShapeTransform(Control=@n13.Item.Name) {is=struct RigVMUnitNode}.Transform.Scale3D
    Item=@n13.Item
   ) {template="DISPATCH_RigDispatch_SetMetadata(in Item,in Name,in NameSpace,in Value,out Success)"}
   > SpawnScaleAnimationChannel = &RigUnit_HierarchyAddAnimationChannelScaleFloat(
    InitialValue=1
    LimitsEnabled=(Enabled=(bMinimum=True,bMaximum=True))
    MaximumValue=20
    MinimumValue=0
    Name=Control Display Scale
    Parent=@n13.Item
   ) {template="SpawnScaleAnimationChannel::Execute(in InitialValue,in MinimumValue,in MaximumValue,in LimitsEnabled,in Parent,in Name,out Item)"}
   Concat = /Script/RigVM.RigVMFunction_NameConcat(B=_Direct, A=@n13.Item.Name) {is=struct RigVMUnitNode, template="Concat::Execute(in A,in B,out Result)"}
   > n14 = &RigUnit_HierarchyAddControlTransform(
    InitialValue=(Rotation=(X=0,Y=0,Z=0,W=1),Translation=(X=0,Y=0,Z=0),Scale3D=(X=1,Y=1,Z=1))
    OffsetSpace=LocalSpace
    OffsetTransform=(Rotation=(X=0,Y=0,Z=0,W=1),Translation=(X=0,Y=0,Z=0),Scale3D=(X=1,Y=1,Z=1))
    Name=@Concat.Result
    Parent=@n13.Item
    Settings=RigVMVariableNode() {variable=Direct Control Settings}.Value
   ) {template="SpawnControl::Execute(in InitialValue,in Settings,in OffsetTransform,in OffsetSpace,in Parent,in Name,out Item)"}
   > RigVMDispatchNode(Name=Control, NameSpace=Self, Item=@n8.Element, Value=@n14.Item) {template="DISPATCH_RigDispatch_SetMetadata(in Item,in Name,in NameSpace,in Value,out Success)"}
   > RigVMDispatchNode(
    Name=Control Display Scale Init
    NameSpace=Self
    Value=&RigUnit_GetShapeTransform(Control=@n14.Item.Name).Transform.Scale3D
    Item=@n14.Item
   ) {template="DISPATCH_RigDispatch_SetMetadata(in Item,in Name,in NameSpace,in Value,out Success)"}
   > &RigUnit_HierarchyAddAnimationChannelScaleFloat(
    InitialValue=1
    LimitsEnabled=(Enabled=(bMinimum=True,bMaximum=True))
    MaximumValue=20
    MinimumValue=0
    Name=Control Display Scale
    Parent=@n14.Item
   ) {template="SpawnScaleAnimationChannel::Execute(in InitialValue,in MinimumValue,in MaximumValue,in LimitsEnabled,in Parent,in Name,out Item)"}
   RigUnit_BeginExecution = &RigUnit_BeginExecution()
   n15 = RigVMDispatchNode(ExecuteContext=@RigUnit_BeginExecution.ExecutePin, Array=RigVMVariableNode() {variable=RiggedJoints}.Value) {template="DISPATCH_RigVMDispatch_ArrayIterator(in Array,out Element,out Index,out Count,out Ratio)"}
   DISPATCH_RigDispatch_GetMetadata = RigVMDispatchNode(Default=(Type=None,Name="None"), Name=Control, NameSpace=Self, Item=@n15.Element) {template="DISPATCH_RigDispatch_GetMetadata(in Item,in Name,in NameSpace,in Default,out Value,out Found)"}
   Branch_3_RigVMUnitNode = &RigVMFunction_ControlFlowBranch(ExecuteContext=@n15.ExecuteContext, Condition=@DISPATCH_RigDispatch_GetMetadata.Found)
   &RigUnit_SetTransform(
    Weight=1
    bInitial=false
    bPropagateToChildren=true
    Item=@n15.Element
    ExecutePin=@Branch_3_RigVMUnitNode.True
    Value=&RigUnit_GetTransform(bInitial=false, Item=@DISPATCH_RigDispatch_GetMetadata.Value).Transform
   )
   n16 = &RigUnit_HierarchyGetParent(bDefaultParent=True, Child=@DISPATCH_RigDispatch_GetMetadata.Value)
   n17 = RigVMVariableNode() {variable=GlobalSettingsCtrl}
   n18 = /Script/ControlRig.RigUnit_GetFloatAnimationChannel(Channel=Control Display Scale, bInitial=False, Control=@n17.Value.Name) {is=struct RigVMUnitNode, template="GetAnimationChannel::Execute(out Value,in Control,in Channel,in bInitial)"}
   n19 = &RigVMFunction_MathDoubleMul(
    A=&RigUnit_GetFloatAnimationChannel(Channel=Control Display Scale, bInitial=False, Control=@n16.Parent.Name) {template="GetAnimationChannel::Execute(out Value,in Control,in Channel,in bInitial)"}.Value
    B=@n18.Value
   )
   > /Script/ControlRig.RigUnit_SetShapeTransform(
    Transform=(Rotation=(X=0,Y=0,Z=0,W=1),Translation=(X=0,Y=0,Z=0),Scale3D=(X=1,Y=1,Z=1))
    Control=@n16.Parent.Name
    Transform.Scale3D=/Script/RigVM.RigVMFunction_MathVectorMul(
     B=(X=1,Y=1,Z=1)
     A=RigVMDispatchNode(Default=(X=0,Y=0,Z=0), Name=Control Display Scale Init, NameSpace=Self, Item=@n16.Parent) {template="DISPATCH_RigDispatch_GetMetadata(in Item,in Name,in NameSpace,in Default,out Value,out Found)"}.Value
     B.X=@n19.Result
     B.Y=@n19.Result
     B.Z=@n19.Result
    ) {is=struct RigVMUnitNode, template="Multiply::Execute(in A,in B,out Result)"}.Result
   ) {is=struct RigVMUnitNode}
   n20 = /Script/ControlRig.RigUnit_GetBoolAnimationChannel(Channel=Control Visibility, bInitial=False, Control=@n17.Value.Name) {is=struct RigVMUnitNode, template="GetAnimationChannel::Execute(out Value,in Control,in Channel,in bInitial)"}
   > SetControlVisibility = /Script/ControlRig.RigUnit_SetControlVisibility(Item=(Type=Control), bVisible=@n20.Value, Item.Name=@n16.Parent.Name) {is=struct RigVMUnitNode, template="SetControlVisibility::Execute(in Item,in Pattern,in bVisible)"}
   GetAnimationChannel = &RigUnit_GetFloatAnimationChannel(Channel=Control Display Scale, bInitial=False, Control=@DISPATCH_RigDispatch_GetMetadata.Value.Name) {template="GetAnimationChannel::Execute(out Value,in Control,in Channel,in bInitial)"}
   Multiply = &RigVMFunction_MathDoubleMul(B=@GetAnimationChannel.Value, A=@n18.Value)
   > &RigUnit_SetShapeTransform(
    Transform=(Rotation=(X=0,Y=0,Z=0,W=1),Translation=(X=0,Y=0,Z=0),Scale3D=(X=1,Y=1,Z=1))
    Control=@DISPATCH_RigDispatch_GetMetadata.Value.Name
    Transform.Scale3D=&RigVMFunction_MathVectorMul(
     B=(X=1,Y=1,Z=1)
     A=RigVMDispatchNode(Default=(X=0,Y=0,Z=0), Name=Control Display Scale Init, NameSpace=Self, Item=@DISPATCH_RigDispatch_GetMetadata.Value) {template="DISPATCH_RigDispatch_GetMetadata(in Item,in Name,in NameSpace,in Default,out Value,out Found)"}.Value
     B.X=@Multiply.Result
     B.Y=@Multiply.Result
     B.Z=@Multiply.Result
    ) {template="Multiply::Execute(in A,in B,out Result)"}.Result
   )
   RigVMFunction_MathBoolAnd = &RigVMFunction_MathBoolAnd(
    B=&RigUnit_GetBoolAnimationChannel(Channel=Direct Control Visibility, bInitial=False, Control=@n17.Value.Name) {template="GetAnimationChannel::Execute(out Value,in Control,in Channel,in bInitial)"}.Value
    A=@n20.Value
   )
   > &RigUnit_SetControlVisibility(Item=(Type=Control), Item.Name=@DISPATCH_RigDispatch_GetMetadata.Value.Name, bVisible=@RigVMFunction_MathBoolAnd.Result) {template="SetControlVisibility::Execute(in Item,in Pattern,in bVisible)"}
  ]}
 SequenceAggregate_1_ContainedGraph {nodes [
   n1 = RigVMFunctionEntryNode()
   SequenceAggregate = &RigVMFunction_Sequence(ExecuteContext=@n1.ExecuteContext)
   SequenceAggregate_1 = &RigVMFunction_Sequence(ExecuteContext=@SequenceAggregate.B)
   RigVMFunctionReturnNode(A=@SequenceAggregate.A, B=@SequenceAggregate_1.A, C=@SequenceAggregate_1.B)
  ]}
]
variables [
 Control Settings {
  type=FRigUnit_HierarchyAddControlTransform_Settings
  default="(InitialSpace=LocalSpace,bUsePreferredRotationOrder=False,PreferredRotationOrder=YZX,Limits=(LimitTranslationX=(bMinimum=False,bMaximum=False),LimitTranslationY=(bMinimum=False,bMaximum=False),LimitTranslationZ=(bMinimum=False,bMaximum=False),LimitPitch=(b ...(truncated)"
 }
 Direct Control Settings {
  type=FRigUnit_HierarchyAddControlTransform_Settings
  default="(InitialSpace=LocalSpace,bUsePreferredRotationOrder=False,PreferredRotationOrder=YZX,Limits=(LimitTranslationX=(bMinimum=False,bMaximum=False),LimitTranslationY=(bMinimum=False,bMaximum=False),LimitTranslationZ=(bMinimum=False,bMaximum=False),LimitPitch=(b ...(truncated)"
 }
 Global Control Settings {
  type=FRigUnit_HierarchyAddControl_ShapeSettings
  default=(bVisible=False,Name="Default",Color=(R=1,G=0,B=0,A=1),Transform=(Rotation=(X=0,Y=0,Z=0,W=1),Translation=(X=0,Y=0,Z=0),Scale3D=(X=1,Y=1,Z=1)))
 }
 GlobalSettingsCtrl {type=FRigElementKey, default=(Type=None,Name="")}
 REF_Body_Skeleton_Names {
  type="TArray<FString>"
  default="(\"root\",\"pelvis\",\"spine_01\",\"thigh_l\",\"thigh_r\",\"deform_glute_base_l\",\"deform_glute_base_r\",\"spine_02\",\"calf_l\",\"deform_groin_base_l\",\"calf_r\",\"deform_groin_base_r\",\"deform_glute_l\",\"deform_glute_r\",\"spine_03\",\"foot_l\",\"deform_knee_in_base_l\",\"deform_knee_ ...(truncated)"
 }
 REF_Face_Skeleton_Names {
  type="TArray<FString>"
  default="(\"faceAttach\",\"C_brow_mid\",\"C_jaw\",\"C_lip_upper_mid\",\"C_nose_bridge\",\"L_brow_mid\",\"L_brow_outer\",\"L_cheek_inner\",\"L_eye\",\"L_lip_corner\",\"L_lip_upper_outer\",\"R_brow_mid\",\"R_brow_outer\",\"R_cheek_inner\",\"R_eye\",\"R_lip_corner\",\"R_lip_upper_outer\",\"teeth_upper\" ...(truncated)"
 }
 RiggedJoints {type="TArray<FRigElementKey>", default=(())}
 exclude_joints {type="TArray<FString>"}
 include_joints {type="TArray<FString>"}
]
